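; Gameplay ability for a Battle Pass narrative quest that reports when the player is on a sportbike in a valid location.
; On event activation it checks the pawn's vehicle for the sportbike tag, then tests the pawn against each valid location tag.
; A match is saved, added to the event source tags and sent as a complex custom stat event, with a debug print.
; The temporary tag is then removed and the ability ends.
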(event Ability|EventActivateAbilityFromEvent
  (params (EventData :type Gameplay%20Event%20Data%20Structure%20%28by%20ref%29))
  (bind _eventsourcetags (Variables|Default|GetEventSourceTags))
  (bind _currentlocationtag (Variables|Default|GetCurrentLocationTag))
  ;[0 0 0] Check if player gets on a sportbike
  (|CallParentFunction EventData)
  (bind _asfort_player_pawn (Utilities|Casting|CastToFortPlayerPawn (Ability|GetAvatarActorfromActorInfo))
    (:then
      (Variables|Default|SetOwningPlayerPawn _asfort_player_pawn)
      (bind _returnvalue (Class|FortVehicleInterface|GetGameplayTags(Message) (Vehicle|GetVehicleActor _asfort_player_pawn :!class /Script/FortniteGame.FortPlayerPawn)))
      (if (GameplayTags|HasTag _returnvalue (Variables|Default|GetSportBikeTag) true)
        ;;[0.1406 0.1406 0.1406] Check if player is in valid location; if so, save off the tag and send it with a custom stat event
        (bind _array_element (Utilities|Array|ForEachLoopwithBreak (GameplayTags|BreakGameplayTagContainer (Variables|Default|GetValidLocationTags)))
          (:LoopBody
            (if (GameplayTags|HasMatchingGameplayTag (Variables|Default|GetOwningPlayerPawn) _array_element)
              (bind _output_get (Variables|Default|SetCurrentLocationTag _array_element))
              (GameplayTags|AddGameplayTag (Variables|Default|GetEventSourceTags) _output_get :!class /Script/GameplayTags.BlueprintGameplayTagLibrary)
              (Quest|SendComplexCustomStatEvent (Quests|GetQuestManager (Fortnite|GetFortPlayerControllerfromActor (Variables|Default|GetOwningPlayerPawn)) "Invalid") 0 (Variables|Default|GetEventSourceTags) "(GameplayTags=)" 1)
              (Development|PrintString "Valid location, sending event..." true true "(R=1.000000,G=0.000000,B=0.210374,A=1.000000)")))
          (:Completed
            _))
        (if (GameplayTags|HasTag _eventsourcetags _currentlocationtag true)
          (GameplayTags|RemoveGameplayTag _eventsourcetags _currentlocationtag :!class /Script/GameplayTags.BlueprintGameplayTagLibrary)
          (else
            _))
        (Ability|EndAbility)
        (else
          _)))
    (:CastFailed
      _))
  (Ability|EndAbility))
